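# Niagara effect for a controller spawning in the CajunTaco emote, combining a short puff burst with fading colored trails.
# One-shot emitters layer an air glow, curl-noise sparkles, toon mesh smoke and impact smoke sprites, with smoke tinted by time of day.
# ContrailsOpt samples the system's skeletal mesh, using spawn groups and visibility tags to feed ribbon and four-color sprite renderers.
# Loop-age curves shape trail width, lifetime and alpha before an age-based kill.
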NiagaraSystem /BRCosmetics/Animation/Game/MainPlayer/Emotes/CajunTaco/FX/NS_Emote_Controller_Spawn
dependencies {
 usedDataInterfaces [
  &NiagaraDataInterfaceCurve
  &NiagaraDataInterfaceArrayFloat
  &NiagaraDataInterfaceArrayInt32
  &NiagaraDataInterfaceArrayColor
  &NiagaraDataInterfaceSkeletalMesh
 ]
 usedDynamicInputs []
 usedModules [
  &SystemState
  &self:SystemSpawnScript.SystemScriptSource.SystemScriptGraph.NiagaraNodeAssignment_0.SetVariables_#1
  &self:SystemSpawnScript.SystemScriptSource.SystemScriptGraph.NiagaraNodeAssignment_1.SetVariables_#2
 ]
}
emitters {
 AirGlow {
  emitterUpdateScript [
   &EmitterState()
   &SpawnBurst_Instantaneous(Spawn Count=1, Use Spawn Probability=false)
  ]
  particleSpawnScript [
   &InitializeParticle(
    Color=<rgba 0.4323 0.4323 0.4323 0.2>
    Lifetime=0.25
    Sprite Size Mode=::NewEnumerator3
    Uniform Sprite Size=70
    UsePositionOffset=false
    Write Material Random=true
   ) {module=InitializeParticle001}
   &DynamicMaterialParameters(
    Index 0 Param 1=2
    Module.Param0WriteEnabled=true
    Module.Param1WriteEnabled=true
    Module.Param2WriteEnabled=true
    Module.Param3WriteEnabled=true
   )
   &CameraOffset(Camera Offset Amount=10)
  ]
  particleUpdateScript [
   &ParticleState()
   &ScaleSpriteSize(Scale Factor=&Vector2DFromFloat(Value=&FloatFromCurve(FloatCurve=DataInterface() {propertyValues {Curve {keys [<at 0 1 RCIM_Cubic>]}}})))
   &ScaleColor(
    Scale Alpha=&FloatFromCurve(FloatCurve=DataInterface() {propertyValues {Curve {keys [<at 0 0>,<at 0.25 0.75>,<at 1 0>]}}})
    ScaleA=true
    ScaleRGB=true
   ) {module=ScaleColor001}
  ]
  rendererClasses [&NiagaraSpriteRendererProperties]
  renderers [{rendererClass=&NiagaraSpriteRendererProperties}]
 }
 ContrailsOpt {
  emitterUpdateScript [
   &EmitterState(
    Life Cycle Mode=::NewEnumerator1
    Loop Behavior=::NewEnumerator1
    Loop Delay=0.07
    Loop Duration=2
    Scalability Mode=::NewEnumerator1
    UseLoopDelay=false
   )
   &SpawnRate(SpawnRate=256, Use Spawn Probability=false)
   &SpawnRate(
    SpawnGroup=1
    SpawnRate=&FloatFromCurve(FloatCurve=DataInterface() {propertyValues {Curve {keys [<at 0.6 1>,<at 0.8 0>]}}}, CurveIndex=$Emitter.NormalizedLoopAge, Scale Curve=15)
    Use Spawn Probability=false
   ) {module=SpawnRate001}
  ]
  particleSpawnScript [
   /SetVariables(Particles.VisibilityTag=&SelectIntFromArray(Direct Array Index=&SpawnGroup(), Int Selection Array=DataInterface() {propertyValues {IntData [0,1]}})) {isSetParameters=true}
   &FN_SetSpawnGroup(Number of Spawn Groups=1)
   &InitializeParticle(
    Sprite Rotation Mode=::NewEnumerator2
    Sprite Size Mode=::NewEnumerator3
    Uniform Sprite Size=&RandomRangeFloat(Minimum=12, Maximum=20)
    UsePositionOffset=false
    Write Material Random=true
   )
   &InitializeRibbon(
    Color=&MultiplyLinearColor(
     Linear Color=$System.Contrails
     Scale Factor=&MakeVector4(
      X=1
      Y=1
      Z=1
      W=&FloatFromCurve(FloatCurve=DataInterface() {propertyValues {Curve {keys [<at 0.02 0>,<at 0.4 0.7>,<at 1 0>]}}}, CurveIndex=$Emitter.NormalizedLoopAge)
     )
    )
    Lifetime=&FloatFromCurve(FloatCurve=DataInterface() {propertyValues {Curve {keys [<at 0 0.2>,<at 0.1 0.4>,<at 1.5 0>]}}}, CurveIndex=$Emitter.NormalizedLoopAge)
    Ribbon Facing Mode:ENiagara_UnsetDirectSet=::NewEnumerator1
    Ribbon Width=&FloatFromCurve(
     FloatCurve=DataInterface() {propertyValues {Curve {keys [<at 0.02 1>,<at 0.25 1>,<at 1 1>]}}}
     CurveIndex=$Emitter.NormalizedLoopAge
     Scale Curve=15
    )
    UsePositionOffset=false
    Write Material Random=true
   )
   &SkeletalMeshLocation(
    Bone / Socket Index=$Particles.SpawnGroup
    MaskSpawnGroup=false
    Mesh Sampling Type=::NewEnumerator5
    Orientation Sampling=::NewEnumerator1
    Sampled Position Offset=<0 0 -2>
    Skeletal Mesh=$System.Skeletal Mesh
    Socket Sampling Mode=::NewEnumerator5
   )
   /SetVariables(Particles.Orientation:Vector3f=&RotateVectorByQuaternion(Vector To Rotate=<0 0 1>, Quaternion=$Particles.SkeletalMeshLocation.SampledRotation)) {isSetParameters=true}
   &DynamicMaterialParameters(
    Index 0 Param 2=&RandomRangeFloat(Minimum=0, Maximum=1)
    Module.Index0Float0Write=false
    Module.Index0Float1Write=true
    Module.Index0Float2Write=false
    Module.Index0Float3Write=false
   ) {enabled=false}
   &AddVelocity(Velocity=<200 0 0>) {enabled=false}
   /SetVariables(Particles.SpritePos:NiagaraPosition=&AddVectorToPosition(
     Position=$Particles.SkeletalMeshLocation.SampledPosition
     Vector=&RandomRangeVector(Minimum=<-5 -5 -5>, Maximum=<5 5 5>)
    )) {isSetParameters=true}
  ]
  particleUpdateScript [
   &ParticleState()
   &SpriteRotationRate(Rotation Rate=&RandomRangeFloat(Minimum=-50, Maximum=50))
   &SolveForcesAndVelocity(Clamp Velocity=false, Limit Acceleration=false)
   &ScaleColor(
    Scale Alpha=&FloatFromCurve(
     FloatCurve=DataInterface() {propertyValues {Curve {keys [<at 0.4 1>,<at 0.8 0>]}}}
     CurveIndex=$Emitter.NormalizedLoopAge
     Scale Curve=&FloatFromCurve(
      FloatCurve=DataInterface() {propertyValues {Curve {keys [
          {interpMode=RCIM_Cubic, tangentMode=RCTM_User, time=0, value=0}
          {interpMode=RCIM_Cubic, tangentMode=RCTM_User, time=0.2, value=1}
          {tangentMode=RCTM_User, time=1, value=0}
         ]}}}
      Scale Curve=2
     )
    )
    ScaleA=true
    ScaleRGB=true
   ) {module=ScaleColor001}
   &KillParticles(Kill Particles=&SetBoolByFloatComparison(A=$Emitter.Age, B=3))
   /SetVariables(Particles.SpriteColor=&MakeLinearColorFromVectorAndFloat(
     "Vector (RGB)"=&MakeVectorFromLinearColorRGB(Color=&SelectLinearColorFromArray(
       Array Sampling Mode=::NewEnumerator0
       Direct Array Index=?
       Color Selection Array=DataInterface() {propertyValues {ColorData [<rgba 0.1036 0.2329 1 1>, <rgba 1 0.2669 0.8674 1>, <rgba 1 0.7266 0.2408 1>, <rgba 0.1258 1 0.2087 1>]}}
       Random Seed=0
      ))
     "Float (Alpha)"=&FloatFromCurve(
      FloatCurve=DataInterface() {propertyValues {Curve {keys [
          {interpMode=RCIM_Cubic, tangentMode=RCTM_User, time=0, value=0}
          {interpMode=RCIM_Cubic, tangentMode=RCTM_User, time=0.2, value=1}
          {tangentMode=RCTM_User, time=1, value=0}
         ]}}}
      Scale Curve=2
     )
    )) {isSetParameters=true}
  ]
  rendererClasses [&NiagaraRibbonRendererProperties, &NiagaraSpriteRendererProperties]
  renderers [
   {rendererClass=&NiagaraRibbonRendererProperties}
   {rendererClass=&NiagaraSpriteRendererProperties, rendererIndex=1}
  ]
 }
 Smoke_Toon {
  emitterSpawnScript [/Game/Effects/Niagara/MasterMaterials/ToonShader/Template/NM_ToonShaderParameters(HighLight_Softness=0, Lighting_Hardness=0.8, Metallic_LightingPower=0, RimLight_Color:LinearColor=<rgba 0 0 0 1>, RimLight_Width=0)]
  emitterUpdateScript [
   &EmitterState()
   &SpawnBurst_Instantaneous(Spawn Count=2, Use Spawn Probability=false, UseLoopCountLimit=false)
  ]
  particleSpawnScript [
   &InitializeParticle(
    Color=&FNTimeOfDayColor(
     Night Color=<rgba 0.1039 0.1554 0.3837 1>
     Morning Color=<rgba 0.2448 0.1986 0.1928 1>
     Day Color=<rgba 0.2448 0.2448 0.2448 1>
     Evening Color=<rgba 0.1921 0.1606 0.2124 1>
    )
    Lifetime=&RandomRangeFloat(Minimum=0.3, Maximum=0.6)
    Mesh Scale Mode=::NewEnumerator3
    Mesh Uniform Scale=0.45
    Sprite Size Mode=::NewEnumerator3
    Uniform Sprite Size=3.1789
    UsePositionOffset=false
    Write Material Random=true
   )
   &InitialMeshOrientation(Rotation=&RandomRangeVector(Minimum=<0 0 0>, Maximum=<1 1 1>), Use Rotation Vector=true)
   &ShapeLocation(
    Hemisphere Distribution=<0.5 1>
    MaskSpawnGroup=false
    Override Seed=false
    Rotation Angle=0.75
    Rotation Axis=<0 0 1>
    Rotation Mode=::NewEnumerator1
    Sphere Radius=5
    Sphere Surface Distribution=0.5
    UseOwnerScale=true
   )
   &AddVelocity(
    Module.Use Falloff Distance=false
    Override Seed=false
    Velocity Mode=::NewEnumerator1
    Velocity Origin=$StackContext.Position
    Velocity Speed=&RandomRangeFloat(Minimum=70, Maximum=120)
   )
   &CameraOffset(Camera Offset Amount=&RandomRangeFloat(Minimum=0, Maximum=5))
  ]
  particleUpdateScript [
   &ParticleState()
   &ScaleMeshSize(Scale Factor=&VectorFromFloat(Value=&FloatFromCurve(FloatCurve=DataInterface() {propertyValues {Curve {keys [
          {arriveTangent=5.3301, interpMode=RCIM_Cubic, leaveTangent=5.3301, tangentMode=RCTM_User, time=0, value=0}
          {interpMode=RCIM_Cubic, tangentMode=RCTM_User, time=0.7, value=0.95}
          {tangentMode=RCTM_User, time=1, value=1}
         ]}}})))
   &Drag(Drag=&RandomRangeFloat(Minimum=9, Maximum=13), Use Linear Drag=true, Use Rotational Drag=false)
   &SolveForcesAndVelocity(Clamp Velocity=false, Limit Acceleration=false)
   &DynamicMaterialParameters(
    Index 0 Param 1=&FloatFromCurve(FloatCurve=DataInterface() {propertyValues {Curve {keys [
         {arriveTangent=2.2672, interpMode=RCIM_Cubic, leaveTangent=2.2672, tangentMode=RCTM_User, time=0, value=0}
         {interpMode=RCIM_Cubic, tangentMode=RCTM_User, time=1, value=1}
        ]}}})
    Module.Param0WriteEnabled=true
    Module.Param1WriteEnabled=true
    Module.Param2WriteEnabled=true
    Module.Param3WriteEnabled=true
   )
  ]
  rendererClasses [&NiagaraMeshRendererProperties]
  renderers [{rendererClass=&NiagaraMeshRendererProperties, rendererIndex=0}]
 }
 Sparkles {
  emitterUpdateScript [
   &EmitterState()
   &SpawnBurst_Instantaneous(Spawn Count=2, Use Spawn Probability=false)
  ]
  particleSpawnScript [
   &InitializeParticle(
    Color=<rgba 5 3.3359 0.1578 1>
    Lifetime=&UniformRangedFloat(Minimum=0.3, Maximum=1)
    Sprite Rotation Angle=45
    Sprite Rotation Mode=::NewEnumerator0
    Sprite Size Mode=::NewEnumerator3
    Uniform Sprite Size=&UniformRangedFloat(Minimum=15, Maximum=8)
    UsePositionOffset=false
    Write Material Random=true
   ) {module=InitializeParticle001}
   &AddVelocity(Velocity=&UniformRangedVector(Minimum=<-100 -100 -100>, Maximum=<100 100 100>))
   &ApplyInitialForces()
   &ShapeLocation(MaskSpawnGroup=false, Override Seed=false, Sphere Radius=7, Sphere Surface Distribution=1, UseOwnerScale=true)
   &CameraOffset(Camera Offset Amount=20)
  ]
  particleUpdateScript [
   &ParticleState()
   &ScaleSpriteSize(Uniform Curve Sprite Scale=DataInterface() {propertyValues {Curve {keys [
        {interpMode=RCIM_Cubic, tangentMode=RCTM_User, time=0, value=0.5718}
        {
         arriveTangent=-3.6492
         arriveTangentWeight=0.1689
         interpMode=RCIM_Cubic
         leaveTangent=-3.6492
         leaveTangentWeight=0.9359
         tangentMode=RCTM_User
         time=0.0814
         value=1
        }
        {tangentMode=RCTM_User, time=1, value=0}
       ]}}}) {module=ScaleSpriteSize001}
   &GravityForce(Gravity=<0 0 200>)
   /Niagara/Modules/Update/Forces/CurlNoiseForce(Noise Frequency=0.02, Noise Quality / Cost=::NewEnumerator0, Noise Strength=500, Pan Noise=false)
   &Drag(Drag=5, Use Linear Drag=true, Use Rotational Drag=false)
   &SolveForcesAndVelocity(Clamp Velocity=false, Limit Acceleration=false)
   &ScaleColor(
    Scale Alpha=&FloatFromCurve(FloatCurve=DataInterface() {propertyValues {Curve {keys [<at -0.0092 0.0144>,<at 0.0967 0.9767>,<at 1.0011 0.002>]}}})
    ScaleA=true
    ScaleRGB=true
   )
  ]
  rendererClasses [&NiagaraSpriteRendererProperties]
  renderers [{rendererClass=&NiagaraSpriteRendererProperties}]
 }
 Toon_Impact_Smoke {
  emitterUpdateScript [
   &EmitterState()
   &SpawnBurst_Instantaneous(Spawn Count=1, Spawn Time=0.03, Use Spawn Probability=false)
  ]
  particleSpawnScript [
   &InitializeParticle(
    Color=&FNTimeOfDayColor(
     Night Color=<rgba 0.2149 0.3102 0.842 1>
     Morning Color=<rgba 0.7078 0.5363 0.5148 1>
     Day Color=<rgba 0.8096 0.8096 0.8096 1>
     Evening Color=<rgba 0.6076 0.4758 0.7031 1>
    )
    Lifetime Max=0.35
    Lifetime Min=0.3
    Lifetime Mode=::NewEnumerator1
    Sprite Rotation Angle=&RandomRangeFloat(Minimum=-45, Maximum=45)
    Sprite Rotation Mode=::NewEnumerator0
    Sprite Size Mode=::NewEnumerator1
    Uniform Sprite Size Max=45
    Uniform Sprite Size Min=35
    UsePositionOffset=false
    Write Material Random=true
   ) {module=InitializeParticle001}
   &CameraOffset(Camera Offset Amount=20)
   &DynamicMaterialParameters(
    Index 0 Param 2=&UniformAOrBFloat(A=0)
    Module.Param0WriteEnabled=true
    Module.Param1WriteEnabled=true
    Module.Param2WriteEnabled=true
    Module.Param3WriteEnabled=true
   ) {module=DynamicMaterialParameters001}
  ]
  particleUpdateScript [
   &ParticleState()
   &SolveForcesAndVelocity(Clamp Velocity=false, Limit Acceleration=false)
   &DynamicMaterialParameters(
    Index 0 Param 1=&FloatFromCurve(FloatCurve=DataInterface() {propertyValues {Curve {keys [<at 0 0>,<at 1 1>]}}})
    Index 0 Param 2=1
    Module.Param0WriteEnabled=true
    Module.Param1WriteEnabled=true
    Module.Param2WriteEnabled=true
    Module.Param3WriteEnabled=true
   )
   &ScaleColor(
    Scale Alpha=&FloatFromCurve(FloatCurve=DataInterface() {propertyValues {Curve {keys [<at 0 1>,<at 0.8091 0.9932>,<at 1 0>]}}})
    ScaleA=true
    ScaleRGB=true
   )
  ]
  rendererClasses [&NiagaraSpriteRendererProperties]
  renderers [{rendererClass=&NiagaraSpriteRendererProperties}]
 }
 Toon_Impact_Smoke001 {
  emitterUpdateScript [
   &EmitterState()
   &SpawnBurst_Instantaneous(Spawn Count=3, Use Spawn Probability=false)
  ]
  particleSpawnScript [
   &InitializeParticle(
    Color=&FNTimeOfDayColor(
     Night Color=<rgba 0.2149 0.3102 0.842 1>
     Morning Color=<rgba 0.7078 0.5363 0.5148 1>
     Day Color=<rgba 0.8096 0.8096 0.8096 1>
     Evening Color=<rgba 0.6076 0.4758 0.7031 1>
    )
    Lifetime Max=0.5
    Lifetime Min=0.3
    Lifetime Mode=::NewEnumerator1
    Sprite Rotation Angle=&RandomRangeFloat(Minimum=-45, Maximum=45)
    Sprite Rotation Mode=::NewEnumerator0
    Sprite Size Mode=::NewEnumerator1
    Uniform Sprite Size Max=50
    Uniform Sprite Size Min=70
    UsePositionOffset=false
    Write Material Random=true
   ) {module=InitializeParticle001}
   &ShapeLocation(
    Hemisphere Distribution=<0.5 1>
    MaskSpawnGroup=false
    Override Seed=false
    Rotation Angle=0.75
    Rotation Axis=<0 0 1>
    Rotation Mode=::NewEnumerator1
    Sphere Radius=10
    Sphere Surface Distribution=0.5
    UseOwnerScale=true
   )
   &AddVelocity(
    Module.Use Falloff Distance=false
    Override Seed=false
    Velocity Mode=::NewEnumerator1
    Velocity Origin=$StackContext.Position
    Velocity Speed=&RandomRangeFloat(Minimum=300, Maximum=100)
   )
   &CameraOffset(Camera Offset Amount=20)
  ]
  particleUpdateScript [
   &ParticleState()
   &DynamicMaterialParameters(
    Index 0 Param 1=&FloatFromCurve(FloatCurve=DataInterface() {propertyValues {Curve {keys [
         {arriveTangent=0.5561, interpMode=RCIM_Cubic, leaveTangent=0.5561, tangentMode=RCTM_User, time=0, value=0.2}
         {arriveTangent=0.9173, interpMode=RCIM_Cubic, leaveTangent=0.9173, tangentMode=RCTM_User, time=1, value=1}
        ]}}})
    Index 0 Param 2=&SelectFloatFromArray(
     Array Sampling Mode=::NewEnumerator0
     Direct Array Index=?
     Float Selection Array=DataInterface() {propertyValues {FloatData [0,1]}}
     Random Seed=0
    )
    Module.Index0Float0Write=true
    Module.Index0Float1Write=true
    Module.Index0Float2Write=true
    Module.Index0Float3Write=true
   ) {module=DynamicMaterialParameters001}
   &Drag(Drag=&RandomRangeFloat(Minimum=9, Maximum=13), Use Linear Drag=true, Use Rotational Drag=false)
   &SolveForcesAndVelocity(Clamp Velocity=false, Limit Acceleration=false)
   &ScaleColor(
    Scale Alpha=&FloatFromCurve(FloatCurve=DataInterface() {propertyValues {Curve {keys [<at 0 1>,<at 0.8091 0.9932>,<at 1 0>]}}})
    ScaleA=true
    ScaleRGB=true
   )
  ]
  rendererClasses [&NiagaraSpriteRendererProperties]
  renderers [{rendererClass=&NiagaraSpriteRendererProperties}]
 }
}
summary {
 systemName=NS_Emote_Controller_Spawn
 emitters [
  {emitterName=AirGlow, rendererClasses [&NiagaraSpriteRendererProperties]}
  {emitterName=Sparkles, rendererClasses [&NiagaraSpriteRendererProperties]}
  {emitterName=Smoke_Toon, rendererClasses [&NiagaraMeshRendererProperties]}
  {emitterName=Toon_Impact_Smoke, rendererClasses [&NiagaraSpriteRendererProperties]}
  {emitterName=Toon_Impact_Smoke001, rendererClasses [&NiagaraSpriteRendererProperties]}
  {emitterName=ContrailsOpt, rendererClasses [&NiagaraRibbonRendererProperties, &NiagaraSpriteRendererProperties]}
 ]
 userVariables [User.NormalizedNotifyProgress {type=NiagaraFloat, defaultValue=0, description=""}]
}
